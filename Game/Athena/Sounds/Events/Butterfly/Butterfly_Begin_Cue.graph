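# Sound cue for the start of the Butterfly event in Athena, playing a transition sound on the SFX class with boosted volume.
# A quality-level node at the root selects between two wave players.
# The first two quality slots share the standard transition wave, while the last two share its "_opt" variant.
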SoundCue /Game/Athena/Sounds/Events/Butterfly/Butterfly_Begin_Cue
root=QualityLevel_0
nodes [
 {id=QualityLevel_0, children [WavePlayer_1,WavePlayer_1,WavePlayer_0,WavePlayer_0]}
 {id=WavePlayer_0, properties {soundWaveAssetPtr=&/Butterfly_Transition_01_opt}}
 {id=WavePlayer_1, properties {soundWaveAssetPtr=&/Butterfly_Transition_01}}
]
properties {maxDistance=2097152, soundClassObject=&SFX, volumeMultiplier=1.7}
