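# Sound cue for the Butterfly event, apparently for a camera moment, that plays a rift portal teleport collapse sound on the SFX sound class.
# A single wave player is the root, with no randomization, mixing or modulation nodes.
# Overridden attenuation turns off distance falloff, so the boosted sound keeps the same volume at any distance.
# Concurrency override values specify a small voice cap that stops the oldest instance.
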SoundCue /Game/Athena/Sounds/Events/Butterfly/Butterfly_Cam_Cue
root=WavePlayer_2
nodes [{id=WavePlayer_2, properties {soundWaveAssetPtr=/Game/Sounds/Fort_GamePlay_Sounds/Player/RiftPortal/RiftPortal_Teleport_Collapse_07}}]
properties {
 bOverrideAttenuation=true
 maxDistance=2097152
 soundClassObject=&SFX
 volumeMultiplier=2.25
 attenuationOverrides {bAttenuate=false, falloffDistance=1000, stereoSpread=1000, attenuationShapeExtents {x=0}}
 concurrencyOverrides {maxCount=6, resolutionRule=StopOldest}
}
