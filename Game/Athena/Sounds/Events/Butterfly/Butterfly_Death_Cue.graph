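# Looping SFX cue for the Butterfly death sky sound, blending near and distant layers from one wave.
# A non-spatialized layer fades with distance and mixes with a spatialized, wide-stereo layer whose inverted custom curve brings it in near the falloff edge.
# The distant layer also uses listener focus and a time envelope that swells, then slowly fades.
# All quality levels share the mixer, and concurrency stops the oldest instance.
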SoundCue /Game/Athena/Sounds/Events/Butterfly/Butterfly_Death_Cue
root=QualityLevel_0
nodes [
 {
  id=Attenuation_0
  children [WavePlayer_0]
  properties {bOverrideAttenuation=true, attenuationOverrides {bSpatialize=false, falloffDistance=7000, attenuationShapeExtents {x=0}}}
 }
 {
  id=Attenuation_1
  children [WavePlayer_0]
  properties {
   bOverrideAttenuation=true
   attenuationOverrides {
    bEnableFocusInterpolation=true
    bEnableListenerFocus=true
    distanceAlgorithm=Custom
    falloffDistance=7000
    focusAttackInterpSpeed=0.5
    focusReleaseInterpSpeed=0.5
    nonFocusVolumeAttenuation=0
    stereoSpread=20000
    attenuationShapeExtents {x=0}
    customAttenuationCurve {editorCurveData {keys [<at 0 0>,<at 0.9 0>,<at 1 1>]}}
   }
  }
 }
 {
  id=Enveloper_0
  children [Attenuation_1]
  properties {pitchCurve {editorCurveData {keys [{time=0}]}}, volumeCurve {editorCurveData {keys [<at 0 0>,<at 5 1>,<at 60 0>]}}}
 }
 {id=Mixer_0, children [Attenuation_0, Enveloper_0], properties {inputVolume [1.25,1]}}
 {id=QualityLevel_0, children [Mixer_0,Mixer_0,Mixer_0,Mixer_0]}
 {id=WavePlayer_0, properties {bLooping=true, soundWaveAssetPtr=&/Butterfly_Death_Sky_01}}
]
properties {
 bOverrideConcurrency=true
 maxDistance=2097152
 soundClassObject=&SFX
 volumeMultiplier=1
 attenuationOverrides {bSpatialize=false, falloffDistance=7000, stereoSpread=10000, attenuationShapeExtents {x=0}}
 concurrencyOverrides {maxCount=6, resolutionRule=StopOldest}
}
