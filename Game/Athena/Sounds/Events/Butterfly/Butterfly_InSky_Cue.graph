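# Sound cue for the Athena Butterfly event's in-sky moment, playing the Butterfly_Decay_01 wave on the SFX class after a fixed two-second delay.
# All four quality level inputs share the same delay and wave player branch, so every level sounds identical.
# Overrides turn off distance volume attenuation, set a very wide stereo spread, and cap concurrent instances at six, stopping the oldest.
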SoundCue /Game/Athena/Sounds/Events/Butterfly/Butterfly_InSky_Cue
root=QualityLevel_0
nodes [
 {id=Delay_0, children [WavePlayer_2], properties {delayMax=2, delayMin=2}}
 {id=QualityLevel_0, children [Delay_0,Delay_0,Delay_0,Delay_0]}
 {id=WavePlayer_2, properties {soundWaveAssetPtr=&/Butterfly_Decay_01}}
]
properties {
 bOverrideAttenuation=true
 bOverrideConcurrency=true
 maxDistance=2097152
 soundClassObject=&SFX
 volumeMultiplier=1.3
 attenuationOverrides {bAttenuate=false, stereoSpread=10000}
 concurrencyOverrides {maxCount=6, resolutionRule=StopOldest}
}
