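# Looping SFX cue for the movement of the event Butterfly, layering three reused event and rift portal loops.
# A SplineTime parameter crossfade fades out a double-pitched Geode oddity loop over 0 to 1 while a CubeIsland loop fades in.
# A close rift portal loop joins only near the end of that range.
# An attenuation override disables distance volume falloff.
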SoundCue /Game/Athena/Sounds/Events/Butterfly/Butterfly_Movement_Cue
root=ParamCrossFade_0
nodes [
 {id=Modulator_0, children [WavePlayer_0], properties {pitchMax=2, pitchMin=2}}
 {
  id=ParamCrossFade_0
  children [Modulator_0,WavePlayer_1,WavePlayer_3]
  properties {
   paramName=SplineTime
   crossFadeInput [
    {fadeInDistanceEnd=0, fadeInDistanceStart=0, fadeOutDistanceEnd=1, fadeOutDistanceStart=0, volume=1}
    {fadeInDistanceEnd=1, fadeInDistanceStart=0, fadeOutDistanceEnd=100, fadeOutDistanceStart=100, volume=1}
    {fadeInDistanceEnd=1, fadeInDistanceStart=0.8, fadeOutDistanceEnd=100, fadeOutDistanceStart=100, volume=1}
   ]
  }
 }
 {id=WavePlayer_0, properties {bLooping=true, soundWaveAssetPtr=/Game/Athena/Sounds/Events/Geode/Oddities/Geode_Oddities_InWorldLoop_01}}
 {id=WavePlayer_1, properties {bLooping=true, soundWaveAssetPtr=/Game/Athena/Sounds/Events/CubeIsland/610/DudeBro_Opening_Focus_C_01}}
 {
  id=WavePlayer_3
  properties {bLooping=true, soundWaveAssetPtr=/Game/Sounds/Fort_GamePlay_Sounds/Player/RiftPortal/RiftPortal_InWorldLoop_Close_01}
 }
]
properties {
 bOverrideAttenuation=true
 maxDistance=2097152
 soundClassObject=&SFX
 volumeMultiplier=2.25
 attenuationOverrides {bAttenuate=false, falloffDistance=1000, stereoSpread=1000, attenuationShapeExtents {x=0}}
 concurrencyOverrides {maxCount=6, resolutionRule=StopOldest}
}
