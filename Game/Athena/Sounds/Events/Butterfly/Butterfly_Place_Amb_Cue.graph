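# Sound cue that plays a looping ambient bed for the Butterfly event's place ambience, routed to the SFX sound class.
# A Quality Level node at the root selects one of two looping wave players by audio quality setting.
# The first two quality slots share Butterfly_Amb_Bed_01, and the last two share Butterfly_Amb_Bed_03_opt.
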SoundCue /Game/Athena/Sounds/Events/Butterfly/Butterfly_Place_Amb_Cue
root=QualityLevel_0
nodes [
 {id=QualityLevel_0, children [WavePlayer_0,WavePlayer_0,WavePlayer_1,WavePlayer_1]}
 {id=WavePlayer_0, properties {bLooping=true, soundWaveAssetPtr=&/Butterfly_Amb_Bed_01}}
 {id=WavePlayer_1, properties {bLooping=true, soundWaveAssetPtr=&/Butterfly_Amb_Bed_03_opt}}
]
properties {maxDistance=2097152, soundClassObject=&SFX, volumeMultiplier=1}
