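# Sound cue for the Athena Butterfly event's plop effect, playing the Butterfly_Sequence_01 wave through the SFX class with boosted volume.
# A quality-level node at the root sends the first two tiers to the full wave and the last two to its "_opt" variant.
# An inline attenuation override turns off distance-based volume falloff, so the sound does not fade with range.
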SoundCue /Game/Athena/Sounds/Events/Butterfly/Butterfly_Plop_Cue
root=QualityLevel_0
nodes [
 {id=QualityLevel_0, children [WavePlayer_0,WavePlayer_0,WavePlayer_1,WavePlayer_1]}
 {id=WavePlayer_0, properties {soundWaveAssetPtr=&/Butterfly_Sequence_01}}
 {id=WavePlayer_1, properties {soundWaveAssetPtr=&/Butterfly_Sequence_01_opt}}
]
properties {
 bOverrideAttenuation=true
 maxDistance=2097152
 soundClassObject=&SFX
 volumeMultiplier=2.25
 attenuationOverrides {bAttenuate=false, falloffDistance=1000, stereoSpread=3000, attenuationShapeExtents {x=0}}
 concurrencyOverrides {maxCount=6, resolutionRule=StopOldest}
}
