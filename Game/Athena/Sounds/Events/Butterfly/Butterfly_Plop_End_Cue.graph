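# Butterfly_Plop_End_Cue is an Athena Sound Cue, named for the end plop of a butterfly event, but its graph is empty.
# It has no nodes or root, so it references no audio to play.
# Only settings remain: the SFX sound class, raised volume, and overrides that turn off distance attenuation and define a stop-oldest concurrency limit.
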SoundCue /Game/Athena/Sounds/Events/Butterfly/Butterfly_Plop_End_Cue
root=null
nodes []
properties {
 maxDistance=2097152
 soundClassObject=&SFX
 volumeMultiplier=1.5
 attenuationOverrides {bAttenuate=false, falloffDistance=1000, stereoSpread=1000, attenuationShapeExtents {x=0}}
 concurrencyOverrides {maxCount=6, resolutionRule=StopOldest}
}
